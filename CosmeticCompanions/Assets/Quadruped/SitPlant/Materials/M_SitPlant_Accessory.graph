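# Master material for the SitPlant quadruped cosmetic companion's accessories, adding separate color tints for the jacket and bandana.
# It wraps the shared 2023 character parent PreFX and PostFX functions, using material attributes and a shading model set by expression.
# Channel mask parameters select jacket and bandana masks, defaulting to blue and green.
# Chained lerps blend tinted base color through named reroutes and write it back into the material attributes.
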Material /CosmeticCompanions/Assets/Quadruped/SitPlant/Materials/M_SitPlant_Accessory
graph {
 properties {
  bUseMaterialAttributes=true
  bUsedWithClothing=true
  bUsedWithMeshDeformer=true
  bUsedWithMorphTargets=true
  bUsedWithNanite=true
  bUsedWithNiagaraMeshParticles=true
  bUsedWithSkeletalMesh=true
  pixelDepthOffsetMode=PDOM_Legacy
  shadingModel=MSM_FromMaterialExpression
  usedShadingModels=Default Lit | Subsurface | Clear Coat | Subsurface Profile | Cloth
 }
 nodes [
  n1 = MaterialFunctionCall() {materialFunction=&MF_CharacterParent_2023_PreFX}
  MaterialFunctionCall(Material=NamedRerouteUsage(), M=@n1.M, Pre Skinned Local Position=@n1.M, Distance Blend=@n1.M, Fresnel=@n1.M) {materialFunction=&MF_CharacterParent_2023_PostFX}
  NamedRerouteDeclaration(None=@n1.Tech Art Mask) {name=TechArtMask, nodeColor=<rgba 0.5937 1 0 1>}
  NamedRerouteDeclaration(None=@n1.Result) {name=IN_Material, nodeColor=<rgba 1 0 0.9141 1>}
  NamedRerouteDeclaration(None=GetMaterialAttributes(None=NamedRerouteUsage()) {attributeGetTypes [#1]}.Base Color) {name=IN_BaseColor, nodeColor=<rgba 0 1 0.7656 1>}
  n2 = NamedRerouteUsage()
  NamedRerouteDeclaration(None=ChannelMaskParameter(None=@n2) {group=SitPlant, maskChannel=Blue, parameterName=FXMaskChannel_Jacket, defaultValue=<rgba 0 0 1 0>}.R) {name=Mask_Jacket, nodeColor=<rgba 1 0.4453 0 1>}
  n3 = ChannelMaskParameter(None=@n2) {group=SitPlant, maskChannel=Green, parameterName=FXMaskChannel_Bandana, defaultValue=<rgba 0 1 0 0>}
  NamedRerouteDeclaration(None=@n3.R) {name=Mask_Bandana, nodeColor=<rgba 1 0.4453 0 1>}
  n4 = NamedRerouteUsage()
  n5 = VectorParameter() {group=SitPlant, parameterName=Color_Jacket, defaultValue=<rgba 1 1 1 1>}
  n6 = LinearInterpolate(
   A=LinearInterpolate(A=NamedRerouteUsage(), B=Multiply(A=@n4, B=@n5.RGB), Alpha=NamedRerouteUsage())
   B=Multiply(A=NamedRerouteUsage(), B=VectorParameter() {group=SitPlant, parameterName=Color_Bandana, defaultValue=<rgba 1 1 1 1>}.RGB)
   Alpha=NamedRerouteUsage()
  )
  NamedRerouteDeclaration(None=SetMaterialAttributes!(MaterialAttributes=NamedRerouteUsage(), Base Color=NamedRerouteDeclaration(None=@n6) {name=OUT_BaseColor, nodeColor=<rgba 0.1953 1 0 1>}) {attributeSetTypes [#1]}) {name=OUT_Material, nodeColor=<rgba 1 0 0.5156 1>}
 ]
}
