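# Post-process AnimBlueprint for the SitPlant quadruped cosmetic companion, adjusting the incoming pose on its skeleton.
# The input pose gets an additive sequence from AdditivePoseAnim weighted by CosmeticPoseAlpha, then passes through the DogFace pose asset blend to the output.
# A chain of spring-driven Anim Dynamics nodes for tail_01 to tail_04 exists but is not connected to the output.
# The event graph is empty and neither variable is declared here.
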AnimBlueprint /CosmeticCompanions/Assets/Quadruped/SitPlant/Meshes/SitPlant_Quadruped_Post_AnimBP
target_skeleton=&/SitPlant_Quadruped_Skeleton
graphs [
 AnimGraph {nodes [
   n1 = Anim Dynamics - Bone: tail_01(Alpha=1, ExternalForce="0,0,0") {
    angularDampingOverride=1
    angularSpringConstant=5000
    bAngularSpring=true
    bOverrideAngularDamping=true
    bUseGravityOverride=true
    gravityScale=0.35
    simulationSpace=RootRelative
    alphaBoolBlend {blendOption=Linear}
    boundBone {boneName=tail_01}
    componentLinearAccScale=<1 1 1>
    componentLinearVelScale=<1 1 1>
    gravityOverride {z=-1500}
    physicsBodyDefinitions [{
      boundBone {boneName=tail_01}
      boxExtents=<20 20 20>
      constraintSetup {angularLimitsMax=<30 30 30>, angularLimitsMin=<-30 -30 -30>, angularTarget=<1 0 0>}
      localJointOffset=<1 0 0>
     }]
   }
   n2 = Anim Dynamics - Bone: tail_02(ComponentPose=@n1, Alpha=1, ExternalForce="0,0,0") {
    angularDampingOverride=1
    angularSpringConstant=5000
    bAngularSpring=true
    bOverrideAngularDamping=true
    bUseGravityOverride=true
    gravityScale=0.35
    simulationSpace=RootRelative
    alphaBoolBlend {blendOption=Linear}
    boundBone {boneName=tail_02}
    componentLinearAccScale=<1 1 1>
    componentLinearVelScale=<1 1 1>
    gravityOverride {z=-1500}
    physicsBodyDefinitions [{
      boundBone {boneName=tail_02}
      boxExtents=<20 20 20>
      constraintSetup {angularLimitsMax=<30 30 30>, angularLimitsMin=<-30 -30 -30>, angularTarget=<1 0 0>}
      localJointOffset=<1 0 0>
     }]
   }
   n3 = Anim Dynamics - Bone: tail_03(ComponentPose=@n2, Alpha=1, ExternalForce="0,0,0") {
    angularDampingOverride=1
    angularSpringConstant=5000
    bAngularSpring=true
    bOverrideAngularDamping=true
    bUseGravityOverride=true
    gravityScale=0.35
    simulationSpace=RootRelative
    alphaBoolBlend {blendOption=Linear}
    boundBone {boneName=tail_03}
    componentLinearAccScale=<1 1 1>
    componentLinearVelScale=<1 1 1>
    gravityOverride {z=-1500}
    physicsBodyDefinitions [{
      boundBone {boneName=tail_03}
      boxExtents=<20 20 20>
      constraintSetup {angularLimitsMax=<30 30 30>, angularLimitsMin=<-30 -30 -30>, angularTarget=<1 0 0>}
      localJointOffset=<1 0 0>
     }]
   }
   Anim Dynamics - Bone: tail_04(ComponentPose=@n3, Alpha=1, ExternalForce="0,0,0") {
    angularDampingOverride=1
    angularSpringConstant=5000
    bAngularSpring=true
    bOverrideAngularDamping=true
    bUseGravityOverride=true
    gravityScale=0.35
    simulationSpace=RootRelative
    alphaBoolBlend {blendOption=Linear}
    boundBone {boneName=tail_04}
    componentLinearAccScale=<1 1 1>
    componentLinearVelScale=<1 1 1>
    gravityOverride {z=-1500}
    physicsBodyDefinitions [{
      boundBone {boneName=tail_04}
      boxExtents=<20 20 20>
      constraintSetup {angularLimitsMax=<30 30 30>, angularLimitsMin=<-30 -30 -30>, angularTarget=<1 0 0>}
      localJointOffset=<1 0 0>
     }]
   }
   n4 = Input Pose()
   n5 = Sequence Player!(Sequence:Anim Sequence Base Object Reference=Get AdditivePoseAnim() -> (AdditivePoseAnim:Anim Sequence Object Reference)) -> (Pose)
   n6 = Get CosmeticPoseAlpha() -> (CosmeticPoseAlpha:float)
   Output Pose(Result=DogFace_poses_001_PoseAsset!(SourcePose:Pose Link Structure=Apply Additive(Base=@n4, Additive=@n5, Alpha=@n6) {alphaBoolBlend {blendOption=Linear}}) -> (Pose:Pose Link Structure) {
     is=title AnimGraphNode_PoseBlendNode
     blendOption=Linear
     poseAsset=&/DogFace_poses_001_PoseAsset
    })
  ]}
 EventGraph {links [], nodes []}
]
variables []
